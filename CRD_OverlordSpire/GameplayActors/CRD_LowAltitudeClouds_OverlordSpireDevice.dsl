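; Low-altitude cloud actor for the Overlord Spire device that fades its cloud in and out, taking its preview fade time from the owning device on BeginPlay.
; Fade events scale one timeline's play rate by that time and play it forward (out) or reverse (in), continuing mid-fade if already playing.
; The timeline drives custom primitive data on the cloud mesh.
; A completed fade-out also fades the ambient loop and calls InitializeLightningAudioWatcher with false.
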(event EventBeginPlay
  (bind _asdevice_overlord_spire (Utilities|Casting|CastToDevice_OverlordSpire (Actor|GetOwner)))
  (Variables|Default|SetPreview%20Cloud%20Fade%20Time (Variables|/CRD_OverlordSpire/Device_OverlordSpire.Device_OverlordSpire_C|GetPreviewCloudFadeTime _asdevice_overlord_spire))
  (|CallParentFunction))

(event EventFadeOutCloud
  (Components|Timeline|SetPlayRate (Variables|CRD_LowAltitudeClouds_OverlordSpireDevice|GetCloudFadeTimeline) (/ 1.0 (Variables|Default|GetPreview%20Cloud%20Fade%20Time)))
  (if (Components|Timeline|IsPlaying (Variables|CRD_LowAltitudeClouds_OverlordSpireDevice|GetCloudFadeTimeline))
    (bind (_direction _newtrack_0) (Timeline|CloudFadeTimeline Play)
      (:Update
        (Rendering|Material|SetCustomPrimitiveDataFloat (Variables|Default|GetCloudMesh) 12 _newtrack_0))
      (:Finished
        (if (Utilities|Enum|Equal(Enum) _direction "Forward")
          (Audio|Components|Audio|FadeOut (Variables|Default|GetAmbient_Cloud_Loop) 1.0 0.0 "Linear" :!class /Script/Engine.AudioComponent)
          (CallFunction|InitializeLightningAudioWatcher false))))
    (else
      (Timeline|CloudFadeTimeline PlayFromStart))))

(event EventFadeInCloud
  (Components|Timeline|SetPlayRate (Variables|CRD_LowAltitudeClouds_OverlordSpireDevice|GetCloudFadeTimeline) (/ 1.0 (Variables|Default|GetPreview%20Cloud%20Fade%20Time)))
  (if (Components|Timeline|IsPlaying (Variables|CRD_LowAltitudeClouds_OverlordSpireDevice|GetCloudFadeTimeline))
    (Timeline|CloudFadeTimeline Reverse)
    (else
      (Timeline|CloudFadeTimeline ReverseFromEnd))))

(fn ConstructionScript
  (|CallParentFunction))
