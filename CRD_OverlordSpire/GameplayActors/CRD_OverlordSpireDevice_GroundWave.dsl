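; Ground wave actor for the Overlord Spire slam attack, pairing an expanding donut attack with a Niagara ring effect.
; BeginPlay matches the ring's life to the wave duration.
; A workaround event feeds duration and sizes to the donut component, keeping its scale-over-time curve runtime-tunable.
; ApplyGroundWaveDamage builds an effect spec from the stored instigating ability, sets per-target-type SetByCaller damage from scalable floats, then applies it as a global environment effect.
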(event EventBeginPlay
  (|CallParentFunction)
  (Niagara|SetNiagaraVariable(Float) (Variables|Default|GetNS_Spire_Slam_Wave) "RingLife" (Variables|Default|GetGroundWaveDuration)))

(event Custom|SetNormalizedSizeCurveValues
  (params (Duration :type Float%20%28double-precision%29) (InitialSize :type Float%20%28double-precision%29) (MaxSize :type Float%20%28double-precision%29))
  ; Bit of a workaround so that we can use a variable duration and scale "curve" so that the donut attack scale over time can be tunable at runtime
  (Class|FortDonutAttackComponent|SetNormalizeSizeCurveValues (Variables|Default|GetFortDonutAttack) Duration InitialSize MaxSize))

(fn ApplyGroundWaveDamage
  (params (EffectTargetActor :type Actor%20Object%20Reference))
  ; Apply damage overrides for each possible target type(Player, Building, PBW, Vehicles, Guards, Wildlife, Creatures (Deimos))
  (LocalVariables|SetCurrentEffectSpec (Ability|MakeOutgoingGameplayEffectSpec (Variables|Default|GetInstigatingAbility) "/CRD_OverlordSpire/Abilities/Attacks/GE_OverlordSpireDevice_AppendageSlam_InstantDamage.GE_OverlordSpireDevice_AppendageSlam_InstantDamage_C" 1.0))
  (bind (_keypin _valuepin) (Utilities|Map|ForEachLoop(Map) (Variables|Default|GetDamageSFValuesPerType))
    (:then
      (bind _returnvalue (Ability|GameplayEffect|AssignTagSetbyCallerMagnitude (LocalVariables|GetCurrentEffectSpec) _keypin (ScalableFloat|GetValueatLevelwithContextParam (Actor|GetOwner) _valuepin)))
      (LocalVariables|SetCurrentEffectSpec _returnvalue))
    (:CompletedPin
      (bind _returnvalue_2 (Environment|ApplyGlobalEnvironmentGameplayEffectSpectoActor (Variables|GetEffectTargetActor) (LocalVariables|GetCurrentEffectSpec)))
      (return _returnvalue_2))))

(fn SetInstigatingAbility
  (params (Ability :type Gameplay%20Ability%20Object%20Reference))
  (Variables|Default|SetInstigatingAbility Ability))

(fn ConstructionScript
  (|CallParentFunction))
