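; Projectile context actor for the Overlord Spire device that supplies the scope for Data Registry resolution.
; BuildDataRegistryResolverScope casts the owner to Device_OverlordSpire and returns that device's Resolver Scopes, because data is pushed to that context.
; A failed cast returns an empty scope array flagged false.
; GetCacheContextOverride returns zero, and the event graph and construction script add no behavior.
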(event EventBeginPlay
  (|CallParentFunction) ; This node is disabled and will not be called.\nDrag off pins to build functionality.
)

(event Collision|EventActorBeginOverlap
  (params (OtherActor :type Actor%20Object%20Reference))
  (|CallParentFunction OtherActor) ; This node is disabled and will not be called.\nDrag off pins to build functionality.
)

(event EventTick
  (params (DeltaSeconds :type Float%20%28single-precision%29))
  (|CallParentFunction DeltaSeconds) ; This node is disabled and will not be called.\nDrag off pins to build functionality.
)

(fn BuildDataRegistryResolverScope
  ; Forward the resolver scope to the owning/instigating device (since data is pushed to that context)
  (bind _asdevice_overlord_spire (Utilities|Casting|CastToDevice_OverlordSpire (Actor|GetOwner))
    (:then
      (return (Variables|/CRD_OverlordSpire/Device_OverlordSpire.Device_OverlordSpire_C|GetResolver%20Scopes _asdevice_overlord_spire) 1 true))
    (:CastFailed
      (return (Utilities|Array|MakeArray) 0 false))))

(fn GetCacheContextOverride
  (return 0))

(fn ConstructionScript
  (|CallParentFunction))
