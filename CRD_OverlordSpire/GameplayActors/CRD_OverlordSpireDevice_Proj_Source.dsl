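; Projectile source actor for the Overlord Spire Creative device, extending a parent spire projectile source with a settable projectile count.
; Homing targets must pass the parent check plus the device's team and class restrictions, when that component exists.
; Explosion effect specs get SetByCaller damage magnitudes from a per-type scalable float map.
; It notifies the homing projectile ability when spawning finishes, and clears its timer and destroys itself when the device is disabled.
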(event EventEndPlay
  (params (EndPlayReason :type EEndPlayReason%20Enum :default "Destroyed"))
  (|CallParentFunction EndPlayReason)
  (bind _asdevice_overlord_spire (Utilities|Casting|CastToDevice_OverlordSpire (Variables|Internal|GetSpireActor)))
  (Default|UnbindEventfromOnDisabled _asdevice_overlord_spire (EventDispatchers|CreateEvent :!event Custom|OnSpireDisabled)))

(event Custom|OnSpireDisabled
  (Utilities|Time|ClearandInvalidateTimerbyHandle (Variables|Internal|GetProjectileTimerHandle))
  (Actor|DestroyActor self))

(fn GetProjectileCount
  (return (Variables|Default|GetNumProjectilesToSpawn)))

(fn IsTargetHomingValid
  (params (PlayerPawn :type Fort%20Player%20Pawn%20Object%20Reference) (SourceTransform :type Transform))
  (if (|CallParentFunction PlayerPawn SourceTransform)
    (bind _asdevice_overlord_spire (Utilities|Casting|CastToDevice_OverlordSpire (Variables|Internal|GetSpireActor)))
    (Utilities|IsValid (Variables|/CRD_OverlordSpire/Device_OverlordSpire.Device_OverlordSpire_C|GetCreative_UseRestriction_Component _asdevice_overlord_spire)
      (:Is%20Valid
        (return (and (StateChecks|IsValidTeam (Variables|/CRD_OverlordSpire/Device_OverlordSpire.Device_OverlordSpire_C|GetCreative_UseRestriction_Component _asdevice_overlord_spire) (Variables|GetPlayerPawn)) (StateChecks|IsValidClass (Variables|/CRD_OverlordSpire/Device_OverlordSpire.Device_OverlordSpire_C|GetCreative_UseRestriction_Component _asdevice_overlord_spire) (Variables|GetPlayerPawn)))))
      (:Is%20Not%20Valid
        (return true)))
    (else
      (return false))))

(fn MakeProjEffectContainerToApplyOnExplode
  (params (Owner :type Actor%20Object%20Reference))
  (bind (_targetselection _targetgameplayeffectspecs _ownergameplayeffectspecs _activationcues _impactcues _impactnoiserange _flybynoiserange _boverridechargemagnitude _chargemagnitudeoverridevalue) (Utilities|Struct|BreakFortGameplayEffectContainerSpec (LocalVariables|GetGESpecContainer)))
  (bind _gecontainerspec (|CallParentFunction Owner))
  (bind _output_get (LocalVariables|SetGESpecContainer _gecontainerspec))
  (bind _output_get_2 (LocalVariables|SetModifiedGESpecs (Utilities|Struct|BreakFortGameplayEffectContainerSpec _output_get)))
  (for (_array_element _array_index) _output_get_2
    (LocalVariables|SetGESpecHandle _array_element)
    (bind (_keypin _valuepin) (Utilities|Map|ForEachLoop(Map) (Variables|Default|GetDamageSFValuesPerType))
      (:then
        (bind _returnvalue (Ability|GameplayEffect|AssignTagSetbyCallerMagnitude (LocalVariables|GetGESpecHandle) _keypin (ScalableFloat|GetValueatLevelwithContextParam (Actor|GetOwner) _valuepin)))
        (LocalVariables|SetGESpecHandle _returnvalue))
      (:CompletedPin
        (Utilities|Array|SetArrayElem (LocalVariables|GetModifiedGESpecs) _array_index (LocalVariables|GetGESpecHandle)))))
  (return (Utilities|Struct|MakeFortGameplayEffectContainerSpec _targetselection (LocalVariables|GetModifiedGESpecs) _ownergameplayeffectspecs _activationcues _impactcues _impactnoiserange _flybynoiserange _boverridechargemagnitude _chargemagnitudeoverridevalue)))

(fn OnFinishedSpawningProjectiles
  (bind _asga_overlord_spire_device_homing_projectile (Utilities|Casting|CastToGA_OverlordSpireDevice_HomingProjectile (Variables|Internal|GetFireAbilityRef))
    (:then
      (Class|GAOverlordSpireDeviceHomingProjectile|OnSourceFinishedSpawningProjectiles _asga_overlord_spire_device_homing_projectile))
    (:CastFailed
      _))
  (|CallParentFunction))

(fn SetOwningSpire
  (params (Spire :type BGA%20Large%20Spire%20Object%20Reference))
  (|CallParentFunction Spire)
  (bind _asdevice_overlord_spire (Utilities|Casting|CastToDevice_OverlordSpire (Variables|Internal|GetSpireActor)))
  (Default|BindEventtoOnDisabled _asdevice_overlord_spire (EventDispatchers|CreateEvent :!event OnSpireDisabled)))

(fn SetProjectileCountOverride
  (params (NumProjectiles :type Integer))
  (Variables|Default|SetNumProjectilesToSpawn NumProjectiles))

(fn ConstructionScript
  (|CallParentFunction))
