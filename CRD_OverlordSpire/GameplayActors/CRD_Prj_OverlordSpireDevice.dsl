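; Homing projectile for the Overlord Spire device that seeks aimed-at targets and filters them by the owning device's restrictions.
; Target search queries the owner's player controller, switching homing width and height after the initial check, then updates the homing target from validated results.
; Validation extends the parent's check with team and class tests from the owner's Creative use-restriction component, if present.
; Max speed comes from a scalable float evaluated against the owner.
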(event EventBeginPlay
  (|CallParentFunction) ; This node is disabled and will not be called.\nDrag off pins to build functionality.
)

(event Collision|EventActorBeginOverlap
  (params (OtherActor :type Actor%20Object%20Reference))
  (|CallParentFunction OtherActor) ; This node is disabled and will not be called.\nDrag off pins to build functionality.
)

(event EventTick
  (params (DeltaSeconds :type Float%20%28single-precision%29))
  (|CallParentFunction DeltaSeconds) ; This node is disabled and will not be called.\nDrag off pins to build functionality.
)

(fn FindandSetHomingTarget
  (bind _returnvalue (CallFunction|GetPotentialAimedatTargets (Fortnite|GetFortPlayerControllerfromActor (Actor|GetOwner)) (Math|Float|SelectFloat (Variables|Homing|GetHomingWidthAndHeight) (Variables|Homing|GetSeekingHomingWidthAndHeight) (Variables|Default|GetInitialCheckForTarget)) (Math|Float|SelectFloat (Variables|Homing|GetHomingWidthAndHeight) (Variables|Homing|GetSeekingHomingWidthAndHeight) (Variables|Default|GetInitialCheckForTarget)) (Variables|Homing|GetHomingRange) (Utilities|Array|MakeArray "ObjectTypeQuery3" "ObjectTypeQuery7") (Variables|Homing|GetActorsToIgnoreForTargeting)))
  (Variables|Default|SetInitialCheckForTarget false)
  (bind _array_element (Utilities|Array|ForEachLoopwithBreak _returnvalue)
    (:LoopBody
      (Utilities|IsValid _array_element
        (:Is%20Valid
          (if (CallFunction|IsValidHomingTarget _array_element)
            (CallFunction|UpdateHomingTarget _array_element)
            (Default|SetHomingVars))
          (CallFunction|UpdateHomingTarget (LocalVariables|GetNullActor)))))))

(fn GetMaxSpeed
  (return (ScalableFloat|GetValueatLevelwithContextParam (Actor|GetOwner) (Variables|Default|GetProjectileMaxSpeedRow))))

(fn IsValidHomingTarget
  (params (Target :type Actor%20Object%20Reference))
  (bind _returnvalue (Actor|GetOwner))
  (bind _returnvalue_2 (Actor|GetComponentbyClass _returnvalue "/Game/Creative/Devices/Common/Components/Creative_UseRestriction_Component.Creative_UseRestriction_Component_C"))
  (if (|CallParentFunction Target)
    (Utilities|IsValid _returnvalue_2
      (:Is%20Valid
        (return (and (StateChecks|IsValidTeam _returnvalue_2 (Variables|GetTarget)) (StateChecks|IsValidClass _returnvalue_2 (Variables|GetTarget)))))
      (:Is%20Not%20Valid
        (return true)))
    (else
      (return false))))

(fn ConstructionScript
  (|CallParentFunction))
