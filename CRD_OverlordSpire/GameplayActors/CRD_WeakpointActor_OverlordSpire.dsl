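; Weakpoint actor for the Overlord Spire boss, a child Blueprint that ties inherited weakpoint behavior to its owning BGA_Large_Spire.
; SetSpireOwner stores the spire, and EnableWeakpoint calls the parent only after that spire finishes its intro sequence.
; ResetWeakpointHealth resolves max health from a scalable float using the spire as context.
; A positive value refills health and forces the weakpoint active post-intro; otherwise it processes health depletion.
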(event EventBeginPlay
  (|CallParentFunction) ; This node is disabled and will not be called.\nDrag off pins to build functionality.
)

(event Collision|EventActorBeginOverlap
  (params (OtherActor :type Actor%20Object%20Reference)))

(event EventTick
  (params (DeltaSeconds :type Float%20%28single-precision%29)))

(fn EnableWeakpoint
  (if (Variables|/SpireBoss/GameplayActors/BGA_Large_Spire.BGA_Large_Spire_C|GetbFinishedIntroSequence (Variables|Default|GetSpireBGA))
    (|CallParentFunction)))

(fn ResetWeakpointHealth
  (bind _weakpoint_maxhealth (Variables|Tunable|GetWeakpoint_MaxHealth))
  (bind _spirebga (Variables|Default|GetSpireBGA))
  (bind _returnvalue (ScalableFloat|GetValueatLevelwithContextParam _spirebga _weakpoint_maxhealth))
  (if (> _returnvalue 0)
    (Variables|Internal|SetWeakpoint_CurrentHealth _returnvalue)
    (bind _spirebga_2 (Variables|Default|GetSpireBGA)
      (:then
        (if (Variables|/SpireBoss/GameplayActors/BGA_Large_Spire.BGA_Large_Spire_C|GetbFinishedIntroSequence _spirebga_2)
          (Variables|Internal|SetbForcedActive true)))
      (:else))
    (else
      (CallFunction|ProcessHealthDepleted))))

(fn SetSpireOwner
  (params (InSpireBGA :type BGA%20Large%20Spire%20Object%20Reference))
  (Variables|Default|SetSpireBGA InSpireBGA))

(fn ConstructionScript
  (|CallParentFunction))
